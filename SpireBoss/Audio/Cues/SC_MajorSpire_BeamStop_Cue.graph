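# Sound cue for the beam stop sound of the Major Spire boss, routed to a player-critical boss fight sound class.
# A mixer layers the beam stop wave, passed through a modulator, with a "sweet stop" wave at a higher input volume.
# It uses the Major Spire attack attenuation, and an owner-limited concurrency override allows one instance at a time, blocking new plays while it is active.
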SoundCue /SpireBoss/Audio/Cues/SC_MajorSpire_BeamStop_Cue
root=Mixer_0
nodes [
 {id=Mixer_0, children [Modulator_0,WavePlayer_0], properties {inputVolume [1,2]}}
 {id=Modulator_0, children [WavePlayer_1]}
 {id=WavePlayer_0, properties {soundWaveAssetPtr=/SpireBoss/Audio/Waves/MajorSpire_Beam_Start_Sweet_Stop}}
 {id=WavePlayer_1, properties {soundWaveAssetPtr=/SpireBoss/Audio/Waves/MajorSpire_Beam_Stop}}
]
properties {
 attenuationSettings=/SpireBoss/Audio/Settings/ATT_MajorSpire_Attacks
 bOverrideConcurrency=true
 maxDistance=17000
 soundClassObject=/Game/Sounds/FortSoundClassesAndModes/Fort_BossFight_SoundClass/Fort_BossFight_PlayerCritical
 volumeMultiplier=2
 concurrencyOverrides {bLimitToOwner=true, maxCount=1, resolutionRule=PreventNew}
}
