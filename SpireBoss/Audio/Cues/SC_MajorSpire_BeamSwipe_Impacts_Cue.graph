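# Sound cue for the Major Spire boss's beam swipe impacts, layering ground-hit and beam sounds.
# A mixer blends a one-shot ground impact, a one-shot beam start, and looping ground-impact and beam layers, with boosted per-input levels.
# Envelopers fade both loops in from silence, the beam loop reaching full volume sooner than the ground loop.
# It uses the beam-impact attenuation asset and a player-critical boss-fight sound class.
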SoundCue /SpireBoss/Audio/Cues/SC_MajorSpire_BeamSwipe_Impacts_Cue
root=Mixer_0
nodes [
 {
  id=Enveloper_0
  children [WavePlayer_2]
  properties {pitchCurve {editorCurveData {keys [{time=0}]}}, volumeCurve {editorCurveData {keys [<at 0 0>,<at 1 1>]}}}
 }
 {
  id=Enveloper_1
  children [WavePlayer_0]
  properties {pitchCurve {editorCurveData {keys [{time=0}]}}, volumeCurve {editorCurveData {keys [<at 0 0>,<at 0.2749 0.3299>,<at 0.5 1>]}}}
 }
 {id=Mixer_0, children [WavePlayer_1,Enveloper_0,Enveloper_1,WavePlayer_4], properties {inputVolume [2,2,1.6,3.5]}}
 {id=WavePlayer_0, properties {bLooping=true, soundWaveAssetPtr=/SpireBoss/Audio/Waves/MajorSpire_Beam_Sweet_Loop}}
 {id=WavePlayer_1, properties {soundWaveAssetPtr=/SpireBoss/Audio/Waves/MajorSpire_Beam_Impt_Ground}}
 {id=WavePlayer_2, properties {bLooping=true, soundWaveAssetPtr=/SpireBoss/Audio/Waves/MajorSpire_Beam_Impt_Ground_Loop}}
 {id=WavePlayer_4, properties {soundWaveAssetPtr=/SpireBoss/Audio/Waves/MajorSpire_Beam_Start_Sweet_Start}}
]
properties {
 attenuationSettings=/SpireBoss/Audio/Settings/ATT_MajorSpire_Beam_Impact
 maxDistance=5500
 soundClassObject=/Game/Sounds/FortSoundClassesAndModes/Fort_BossFight_SoundClass/Fort_BossFight_PlayerCritical
 volumeMultiplier=3
}
