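# Sound cue for the Major Spire boss's beam fire attack.
# A Random node picks one of four equally weighted beam-fire wave variations, then a Modulator applies slight downward pitch variation.
# It uses a shared Major Spire attack attenuation asset and priority concurrency set, and routes to the player-critical boss fight sound class with boosted volume.
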SoundCue /SpireBoss/Audio/Cues/SC_MajorSpire_Beam_Fire
root=Modulator_0
nodes [
 {id=Modulator_0, children [Random_1], properties {pitchMax=1, pitchMin=0.98}}
 {
  id=Random_1
  children [WavePlayer_3,WavePlayer_4,WavePlayer_5,WavePlayer_6]
  properties {bShouldExcludeFromBranchCulling=true, weights [1,1,1,1]}
 }
 {id=WavePlayer_3, properties {soundWaveAssetPtr=/SpireBoss/Audio/Waves/MajorSpire_Beam_Fire_04}}
 {id=WavePlayer_4, properties {soundWaveAssetPtr=/SpireBoss/Audio/Waves/MajorSpire_Beam_Fire_03}}
 {id=WavePlayer_5, properties {soundWaveAssetPtr=/SpireBoss/Audio/Waves/MajorSpire_Beam_Fire_02}}
 {id=WavePlayer_6, properties {soundWaveAssetPtr=/SpireBoss/Audio/Waves/MajorSpire_Beam_Fire_01}}
]
properties {
 attenuationSettings=/SpireBoss/Audio/Settings/ATT_MajorSpire_Attacks
 maxDistance=17000
 priority=2
 soundClassObject=/Game/Sounds/FortSoundClassesAndModes/Fort_BossFight_SoundClass/Fort_BossFight_PlayerCritical
 volumeMultiplier=2.3
 attenuationOverrides {bAttenuate=false, falloffDistance=10000, attenuationShapeExtents {x=0}}
 concurrencyOverrides {maxCount=2, resolutionRule=StopOldest, retriggerTime=0.5, voiceStealReleaseTime=0.1}
 concurrencySet [/SpireBoss/Audio/Settings/Conc_MajorSpire_Attacks_Pri]
}
