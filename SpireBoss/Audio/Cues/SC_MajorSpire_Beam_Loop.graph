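# Sound cue for the looping laser beam of the Major Spire boss, played through the boss-fight player-critical sound class.
# A looping wave runs through a Modulator that fixes a raised pitch, then an Enveloper that ramps volume in from silence over half a second.
# The cue uses laser attenuation and boosted volume, and feeds sidechain audio buses for spire loops, weapons, music, ambience and footsteps.
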SoundCue /SpireBoss/Audio/Cues/SC_MajorSpire_Beam_Loop
root=Enveloper_1
nodes [
 {
  id=Enveloper_1
  children [Modulator_0]
  properties {pitchCurve {editorCurveData {keys [{time=0}]}}, volumeCurve {editorCurveData {keys [<at 0 0>,<at 0.2749 0.3299>,<at 0.5 1>]}}}
 }
 {id=Modulator_0, children [WavePlayer_1], properties {pitchMax=1.34, pitchMin=1.34, volumeMax=1, volumeMin=1}}
 {id=WavePlayer_1, properties {bLooping=true, soundWaveAssetPtr=/SpireBoss/Audio/Waves/MajorSpire_Beam_Loop}}
]
properties {
 attenuationSettings=/SpireBoss/Audio/Settings/ATT_MajorSpire_Laser
 maxDistance=15400
 priority=5
 soundClassObject=/Game/Sounds/FortSoundClassesAndModes/Fort_BossFight_SoundClass/Fort_BossFight_PlayerCritical
 volumeMultiplier=4
 busSends [
  {audioBus=/SpireSharedAssets/Audio/Modulation/AB_Spires_Loops_SideChain_Vol, minSendLevel=0, sendLevel=0.3, soundSourceBus=null}
  {audioBus=&AB_Sidechain_FortWeapon_Vol, minSendLevel=0, sendLevel=0.4, soundSourceBus=null}
  {audioBus=&AB_Sidechain_FortMusic_Vol, minSendLevel=0, sendLevel=0.2, soundSourceBus=null}
  {audioBus=&AB_Sidechain_FortAmbient_Vol, minSendLevel=0, sendLevel=0.6, soundSourceBus=null}
  {audioBus=/Game/Sounds/Modulation/EnvelopeFollowers/Volume/AB_Sidechain_Footsteps_Vol, minSendLevel=0, sendLevel=0.6, soundSourceBus=null}
 ]
}
