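# Sound cue for the Major Spire boss death, in the player-critical boss-fight sound class.
# Randomly picks one of three equally weighted death waves, then applies slight downward pitch variation through a modulator.
# It feeds sidechain audio buses for spire loops, weapons, music, ambience and footsteps.
# A concurrency override allows one voice, stopping the oldest, and spatialization settings reuse the Minor Spire attacks attenuation asset.
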SoundCue /SpireBoss/Audio/Cues/SC_MajorSpire_Death
root=Modulator_0
nodes [
 {id=Modulator_0, children [Random_0], properties {pitchMax=1, pitchMin=0.98}}
 {id=Random_0, children [WavePlayer_0,WavePlayer_1,WavePlayer_2], properties {bShouldExcludeFromBranchCulling=true, weights [1,1,1]}}
 {id=WavePlayer_0, properties {soundWaveAssetPtr=/SpireBoss/Audio/Waves/MajorSpire_Death_01}}
 {id=WavePlayer_1, properties {soundWaveAssetPtr=/SpireBoss/Audio/Waves/MajorSpire_Death_02}}
 {id=WavePlayer_2, properties {soundWaveAssetPtr=/SpireBoss/Audio/Waves/MajorSpire_Death_03}}
]
properties {
 attenuationSettings=/SpireSharedAssets/Audio/Settings/ATT_MinorSpire_Attacks
 bOverrideConcurrency=true
 maxDistance=16000
 soundClassObject=/Game/Sounds/FortSoundClassesAndModes/Fort_BossFight_SoundClass/Fort_BossFight_PlayerCritical
 volumeMultiplier=2
 attenuationOverrides {bAttenuate=false, falloffDistance=10000, attenuationShapeExtents {x=0}}
 busSends [
  {audioBus=/SpireSharedAssets/Audio/Modulation/AB_Spires_Loops_SideChain_Vol, minSendLevel=0, sendLevel=0.6, soundSourceBus=null}
  {audioBus=&AB_Sidechain_FortWeapon_Vol, minSendLevel=0, sendLevel=0.4, soundSourceBus=null}
  {audioBus=&AB_Sidechain_FortMusic_Vol, minSendLevel=0, sendLevel=0.4, soundSourceBus=null}
  {audioBus=&AB_Sidechain_FortAmbient_Vol, minSendLevel=0, sendLevel=0.8, soundSourceBus=null}
  {audioBus=/Game/Sounds/Modulation/EnvelopeFollowers/Volume/AB_Sidechain_Footsteps_Vol, minSendLevel=0, sendLevel=0.6, soundSourceBus=null}
 ]
 concurrencyOverrides {maxCount=1, resolutionRule=StopOldest, retriggerTime=0.5, voiceStealReleaseTime=0.1}
}
