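# Sound cue for the Major Spire boss death explosion, played in the player-critical boss-fight sound class with boosted volume.
# An equal-weight Random node picks one of three explosion waves, and a Modulator adds slight pitch variation.
# It borrows the MinorSpire attacks attenuation asset and overrides concurrency to cap overlapping voices, stopping the oldest.
# Bus sends feed sidechain volume buses for spire loops, weapons, music, ambience and footsteps.
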SoundCue /SpireBoss/Audio/Cues/SC_MajorSpire_Death_Explosion
root=Modulator_0
nodes [
 {id=Modulator_0, children [Random_1], properties {pitchMax=1, pitchMin=0.98}}
 {id=Random_1, children [WavePlayer_3,WavePlayer_4,WavePlayer_5], properties {bShouldExcludeFromBranchCulling=true, weights [1,1,1]}}
 {id=WavePlayer_3, properties {soundWaveAssetPtr=/SpireBoss/Audio/Waves/MajorSpire_Death_Explosion_01}}
 {id=WavePlayer_4, properties {soundWaveAssetPtr=/SpireBoss/Audio/Waves/MajorSpire_Death_Explosion_02}}
 {id=WavePlayer_5, properties {soundWaveAssetPtr=/SpireBoss/Audio/Waves/MajorSpire_Death_Explosion_03}}
]
properties {
 attenuationSettings=/SpireSharedAssets/Audio/Settings/ATT_MinorSpire_Attacks
 bOverrideConcurrency=true
 maxDistance=16000
 soundClassObject=/Game/Sounds/FortSoundClassesAndModes/Fort_BossFight_SoundClass/Fort_BossFight_PlayerCritical
 volumeMultiplier=3.2
 attenuationOverrides {bAttenuate=false, falloffDistance=10000, attenuationShapeExtents {x=0}}
 busSends [
  {audioBus=/SpireSharedAssets/Audio/Modulation/AB_Spires_Loops_SideChain_Vol, minSendLevel=0, sendLevel=0.6, soundSourceBus=null}
  {audioBus=&AB_Sidechain_FortWeapon_Vol, minSendLevel=0, sendLevel=0.4, soundSourceBus=null}
  {audioBus=&AB_Sidechain_FortMusic_Vol, minSendLevel=0, sendLevel=0.4, soundSourceBus=null}
  {audioBus=&AB_Sidechain_FortAmbient_Vol, minSendLevel=0, sendLevel=0.8, soundSourceBus=null}
  {audioBus=/Game/Sounds/Modulation/EnvelopeFollowers/Volume/AB_Sidechain_Footsteps_Vol, minSendLevel=0, sendLevel=0.6, soundSourceBus=null}
 ]
 concurrencyOverrides {maxCount=2, resolutionRule=StopOldest, retriggerTime=0.5, voiceStealReleaseTime=0.1}
}
