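# Sound cue for the Major Spire boss's transition from its entry into idle, routed through the boss fight foley sound class.
# A Random node picks one of five equally weighted wave variants and feeds a Modulator that adds slight pitch variation.
# It uses the spire enter attenuation asset and sends to sidechain buses for spire loops, weapons, music, ambience and footsteps.
# A concurrency override caps simultaneous voices, stopping the oldest.
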SoundCue /SpireBoss/Audio/Cues/SC_MajorSpire_EntryToIdle
root=Modulator_0
nodes [
 {id=Modulator_0, children [Random_2], properties {pitchMax=1, pitchMin=0.98}}
 {
  id=Random_2
  children [WavePlayer_9, WavePlayer_10, WavePlayer_11, WavePlayer_12, WavePlayer_13]
  properties {bShouldExcludeFromBranchCulling=true, weights [1,1,1,1,1]}
 }
 {id=WavePlayer_10, properties {soundWaveAssetPtr=/SpireBoss/Audio/Waves/MajorSpire_Enter_To_Idle_04}}
 {id=WavePlayer_11, properties {soundWaveAssetPtr=/SpireBoss/Audio/Waves/MajorSpire_Enter_To_Idle_03}}
 {id=WavePlayer_12, properties {soundWaveAssetPtr=/SpireBoss/Audio/Waves/MajorSpire_Enter_To_Idle_02}}
 {id=WavePlayer_13, properties {soundWaveAssetPtr=/SpireBoss/Audio/Waves/MajorSpire_Enter_To_Idle_01}}
 {id=WavePlayer_9, properties {soundWaveAssetPtr=/SpireBoss/Audio/Waves/MajorSpire_Enter_To_Idle_05}}
]
properties {
 attenuationSettings=/SpireBoss/Audio/Settings/ATT_MajorSpire_Enter
 bOverrideConcurrency=true
 maxDistance=220000
 soundClassObject=/Game/Sounds/FortSoundClassesAndModes/Fort_BossFight_SoundClass/Fort_BossFight_Foley
 volumeMultiplier=1.0024
 attenuationOverrides {bAttenuate=false, falloffDistance=10000, attenuationShapeExtents {x=0}}
 busSends [
  {audioBus=/SpireSharedAssets/Audio/Modulation/AB_Spires_Loops_SideChain_Vol, minSendLevel=0, sendLevel=0.3, soundSourceBus=null}
  {audioBus=&AB_Sidechain_FortWeapon_Vol, minSendLevel=0, sendLevel=0.3, soundSourceBus=null}
  {audioBus=&AB_Sidechain_FortMusic_Vol, minSendLevel=0, sendLevel=0.2, soundSourceBus=null}
  {audioBus=&AB_Sidechain_FortAmbient_Vol, minSendLevel=0, sendLevel=0.5, soundSourceBus=null}
  {audioBus=/Game/Sounds/Modulation/EnvelopeFollowers/Volume/AB_Sidechain_Footsteps_Vol, minSendLevel=0, sendLevel=0.5, soundSourceBus=null}
 ]
 concurrencyOverrides {maxCount=2, resolutionRule=StopOldest, retriggerTime=0.5, voiceStealReleaseTime=0.1}
}
