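# Sound cue for the Major Spire hit impact, played as a non-attenuated sound through a 2D hit-notification sound class.
# A Quality Level node sends its first two levels to a six-wave random pool and the other two to a three-wave subset.
# A Modulator above it adds slight pitch and volume variation.
# Overridden concurrency limits voices per owner, stopping the oldest.
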SoundCue /SpireBoss/Audio/Cues/SC_MajorSpire_HitImpact
root=Modulator_0
nodes [
 {id=Modulator_0, children [QualityLevel_0], properties {pitchMax=1, pitchMin=0.98, volumeMin=0.7}}
 {id=QualityLevel_0, children [Random_0,Random_0,Random_1,Random_1]}
 {
  id=Random_0
  children [WavePlayer_0,WavePlayer_1,WavePlayer_2,WavePlayer_3,WavePlayer_4,WavePlayer_5]
  properties {bShouldExcludeFromBranchCulling=true, weights [1,1,1,1,1,1]}
 }
 {id=Random_1, children [WavePlayer_6,WavePlayer_7,WavePlayer_8], properties {weights [1,1,1]}}
 {id=WavePlayer_0, properties {soundWaveAssetPtr=/SpireBoss/Audio/Waves/MajorSpire_Impact_01}}
 {id=WavePlayer_1, properties {soundWaveAssetPtr=/SpireBoss/Audio/Waves/MajorSpire_Impact_02}}
 {id=WavePlayer_2, properties {soundWaveAssetPtr=/SpireBoss/Audio/Waves/MajorSpire_Impact_03}}
 {id=WavePlayer_3, properties {soundWaveAssetPtr=/SpireBoss/Audio/Waves/MajorSpire_Impact_04}}
 {id=WavePlayer_4, properties {soundWaveAssetPtr=/SpireBoss/Audio/Waves/MajorSpire_Impact_05}}
 {id=WavePlayer_5, properties {soundWaveAssetPtr=/SpireBoss/Audio/Waves/MajorSpire_Impact_06}}
 {id=WavePlayer_6, properties {soundWaveAssetPtr=&MajorSpire_Impact_01}}
 {id=WavePlayer_7, properties {soundWaveAssetPtr=&MajorSpire_Impact_02}}
 {id=WavePlayer_8, properties {soundWaveAssetPtr=&MajorSpire_Impact_03}}
]
properties {
 bOverrideConcurrency=true
 maxDistance=2097152
 soundClassObject=&Fort_2DHitNotifications_Instigate_PSM
 volumeMultiplier=0.45
 attenuationOverrides {bAttenuate=false, falloffDistance=10000, attenuationShapeExtents {x=0}}
 concurrencyOverrides {bLimitToOwner=true, maxCount=3, resolutionRule=StopOldest, voiceStealReleaseTime=0.1, volumeScaleAttackTime=0.1}
}
