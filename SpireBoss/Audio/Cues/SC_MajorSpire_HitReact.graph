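# Sound cue for the Major Spire boss's hit reaction, routed to the boss-fight PlayerCritical sound class with attack attenuation settings.
# A quality-level node feeds a slight pitch modulator, choosing from eight random waves in its first two slots and the first four waves in its last two.
# Concurrency allows one voice, stopping the oldest.
# Bus sends feed sidechain buses for spire loops, weapons, music, ambience and footsteps.
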SoundCue /SpireBoss/Audio/Cues/SC_MajorSpire_HitReact
root=Modulator_0
nodes [
 {id=Modulator_0, children [QualityLevel_0], properties {pitchMax=1, pitchMin=0.98}}
 {id=QualityLevel_0, children [Random_0,Random_0,Random_3,Random_3]}
 {
  id=Random_0
  children [WavePlayer_0, WavePlayer_1, WavePlayer_2, WavePlayer_8, WavePlayer_15, WavePlayer_16, WavePlayer_17, WavePlayer_18]
  properties {bShouldExcludeFromBranchCulling=true, weights [1,1,1,1,1,1,1,1]}
 }
 {id=Random_3, children [WavePlayer_19, WavePlayer_20, WavePlayer_21, WavePlayer_22], properties {weights [1,1,1,1]}}
 {id=WavePlayer_0, properties {soundWaveAssetPtr=/SpireBoss/Audio/Waves/MajorSpire_HitReact_01}}
 {id=WavePlayer_1, properties {soundWaveAssetPtr=/SpireBoss/Audio/Waves/MajorSpire_HitReact_02}}
 {id=WavePlayer_15, properties {soundWaveAssetPtr=/SpireBoss/Audio/Waves/MajorSpire_HitReact_05}}
 {id=WavePlayer_16, properties {soundWaveAssetPtr=/SpireBoss/Audio/Waves/MajorSpire_HitReact_06}}
 {id=WavePlayer_17, properties {soundWaveAssetPtr=/SpireBoss/Audio/Waves/MajorSpire_HitReact_07}}
 {id=WavePlayer_18, properties {soundWaveAssetPtr=/SpireBoss/Audio/Waves/MajorSpire_HitReact_08}}
 {id=WavePlayer_19, properties {soundWaveAssetPtr=&MajorSpire_HitReact_01}}
 {id=WavePlayer_2, properties {soundWaveAssetPtr=/SpireBoss/Audio/Waves/MajorSpire_HitReact_03}}
 {id=WavePlayer_20, properties {soundWaveAssetPtr=&MajorSpire_HitReact_02}}
 {id=WavePlayer_21, properties {soundWaveAssetPtr=&MajorSpire_HitReact_03}}
 {id=WavePlayer_22, properties {soundWaveAssetPtr=/SpireBoss/Audio/Waves/MajorSpire_HitReact_04}}
 {id=WavePlayer_8, properties {soundWaveAssetPtr=&MajorSpire_HitReact_04}}
]
properties {
 attenuationSettings=/SpireBoss/Audio/Settings/ATT_MajorSpire_Attacks
 bOverrideConcurrency=true
 maxDistance=17000
 soundClassObject=/Game/Sounds/FortSoundClassesAndModes/Fort_BossFight_SoundClass/Fort_BossFight_PlayerCritical
 volumeMultiplier=2.5
 attenuationOverrides {bAttenuate=false, falloffDistance=10000, attenuationShapeExtents {x=0}}
 busSends [
  {audioBus=/SpireSharedAssets/Audio/Modulation/AB_Spires_Loops_SideChain_Vol, minSendLevel=0, sendLevel=0.6, soundSourceBus=null}
  {audioBus=&AB_Sidechain_FortWeapon_Vol, minSendLevel=0, sendLevel=0.7, soundSourceBus=null}
  {audioBus=&AB_Sidechain_FortMusic_Vol, minSendLevel=0, sendLevel=0.5, soundSourceBus=null}
  {audioBus=&AB_Sidechain_FortAmbient_Vol, minSendLevel=0, sendLevel=0.8, soundSourceBus=null}
  {audioBus=/Game/Sounds/Modulation/EnvelopeFollowers/Volume/AB_Sidechain_Footsteps_Vol, minSendLevel=0, sendLevel=0.8, soundSourceBus=null}
 ]
 concurrencyOverrides {maxCount=1, resolutionRule=StopOldest, retriggerTime=0.5, voiceStealReleaseTime=0.1}
}
