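# Sound cue that plays the Major Spire boss's looping idle ambience.
# A single looping wave player feeds the output directly, using the spire's ambient attenuation and the boss-fight foley sound class.
# A concurrency override limits it to one instance per owner, so idle loops do not stack on the same spire.
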SoundCue /SpireBoss/Audio/Cues/SC_MajorSpire_Idle_Loop
root=WavePlayer_1
nodes [{id=WavePlayer_1, properties {bLooping=true, soundWaveAssetPtr=/SpireBoss/Audio/Waves/MajorSpire_Idle_Loop}}]
properties {
 attenuationSettings=/SpireBoss/Audio/Settings/ATT_MajorSpire_Ambient
 bOverrideConcurrency=true
 maxDistance=13000
 soundClassObject=/Game/Sounds/FortSoundClassesAndModes/Fort_BossFight_SoundClass/Fort_BossFight_Foley
 volumeMultiplier=0.7518
 concurrencyOverrides {bLimitToOwner=true, maxCount=1}
}
