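# Sound cue for the Major Spire boss's long movement sounds, routed to the boss-fight foley sound class with short attack attenuation.
# A modulator wraps a quality-level switch, whose first two levels randomly pick from seven wave variants and last two from a four-variant subset.
# The seven-variant random node is excluded from branch culling.
# A concurrency override caps simultaneous instances per owner.
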SoundCue /SpireBoss/Audio/Cues/SC_MajorSpire_Movement_Long
root=Modulator_0
nodes [
 {id=Modulator_0, children [QualityLevel_0]}
 {id=QualityLevel_0, children [Random_0,Random_0,Random_3,Random_3]}
 {
  id=Random_0
  children [WavePlayer_0, WavePlayer_1, WavePlayer_3, WavePlayer_15, WavePlayer_16, WavePlayer_17, WavePlayer_18]
  properties {bShouldExcludeFromBranchCulling=true, weights [1,1,1,1,1,1,1]}
 }
 {id=Random_3, children [WavePlayer_19, WavePlayer_20, WavePlayer_21, WavePlayer_22], properties {weights [1,1,1,1]}}
 {id=WavePlayer_0, properties {soundWaveAssetPtr=/SpireBoss/Audio/Waves/MajorSpire_Movement_Long_07}}
 {id=WavePlayer_1, properties {soundWaveAssetPtr=/SpireBoss/Audio/Waves/MajorSpire_Movement_Long_06}}
 {id=WavePlayer_15, properties {soundWaveAssetPtr=/SpireBoss/Audio/Waves/MajorSpire_Movement_Long_04}}
 {id=WavePlayer_16, properties {soundWaveAssetPtr=/SpireBoss/Audio/Waves/MajorSpire_Movement_Long_03}}
 {id=WavePlayer_17, properties {soundWaveAssetPtr=/SpireBoss/Audio/Waves/MajorSpire_Movement_Long_02}}
 {id=WavePlayer_18, properties {soundWaveAssetPtr=/SpireBoss/Audio/Waves/MajorSpire_Movement_Long_01}}
 {id=WavePlayer_19, properties {soundWaveAssetPtr=&MajorSpire_Movement_Long_04}}
 {id=WavePlayer_20, properties {soundWaveAssetPtr=&MajorSpire_Movement_Long_03}}
 {id=WavePlayer_21, properties {soundWaveAssetPtr=&MajorSpire_Movement_Long_02}}
 {id=WavePlayer_22, properties {soundWaveAssetPtr=&MajorSpire_Movement_Long_01}}
 {id=WavePlayer_3, properties {soundWaveAssetPtr=/SpireBoss/Audio/Waves/MajorSpire_Movement_Long_05}}
]
properties {
 attenuationSettings=/SpireBoss/Audio/Settings/ATT_MajorSpire_Attacks_Short
 bOverrideConcurrency=true
 maxDistance=11000
 soundClassObject=/Game/Sounds/FortSoundClassesAndModes/Fort_BossFight_SoundClass/Fort_BossFight_Foley
 volumeMultiplier=1.0024
 concurrencyOverrides {bLimitToOwner=true, maxCount=3}
}
