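# Sound cue for medium movement foley of the Major Spire boss, using the boss-fight foley sound class.
# A Modulator feeds a Quality Level switch: its first two slots pick randomly from eight equally weighted waves, the last two from only the first four.
# Short attack attenuation spatializes it, and a concurrency override limits it to one instance per owner.
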SoundCue /SpireBoss/Audio/Cues/SC_MajorSpire_Movement_Med
root=Modulator_0
nodes [
 {id=Modulator_0, children [QualityLevel_0]}
 {id=QualityLevel_0, children [Random_1,Random_1,Random_2,Random_2]}
 {
  id=Random_1
  children [WavePlayer_2, WavePlayer_4, WavePlayer_5, WavePlayer_6, WavePlayer_7, WavePlayer_8, WavePlayer_9, WavePlayer_10]
  properties {bShouldExcludeFromBranchCulling=true, weights [1,1,1,1,1,1,1,1]}
 }
 {id=Random_2, children [WavePlayer_11, WavePlayer_12, WavePlayer_13, WavePlayer_14], properties {weights [1,1,1,1]}}
 {id=WavePlayer_10, properties {soundWaveAssetPtr=/SpireBoss/Audio/Waves/MajorSpire_Movement_Medium_08}}
 {id=WavePlayer_11, properties {soundWaveAssetPtr=/SpireBoss/Audio/Waves/MajorSpire_Movement_Medium_01}}
 {id=WavePlayer_12, properties {soundWaveAssetPtr=/SpireBoss/Audio/Waves/MajorSpire_Movement_Medium_02}}
 {id=WavePlayer_13, properties {soundWaveAssetPtr=/SpireBoss/Audio/Waves/MajorSpire_Movement_Medium_03}}
 {id=WavePlayer_14, properties {soundWaveAssetPtr=/SpireBoss/Audio/Waves/MajorSpire_Movement_Medium_04}}
 {id=WavePlayer_2, properties {soundWaveAssetPtr=&MajorSpire_Movement_Medium_01}}
 {id=WavePlayer_4, properties {soundWaveAssetPtr=&MajorSpire_Movement_Medium_02}}
 {id=WavePlayer_5, properties {soundWaveAssetPtr=&MajorSpire_Movement_Medium_03}}
 {id=WavePlayer_6, properties {soundWaveAssetPtr=&MajorSpire_Movement_Medium_04}}
 {id=WavePlayer_7, properties {soundWaveAssetPtr=/SpireBoss/Audio/Waves/MajorSpire_Movement_Medium_05}}
 {id=WavePlayer_8, properties {soundWaveAssetPtr=/SpireBoss/Audio/Waves/MajorSpire_Movement_Medium_06}}
 {id=WavePlayer_9, properties {soundWaveAssetPtr=/SpireBoss/Audio/Waves/MajorSpire_Movement_Medium_07}}
]
properties {
 attenuationSettings=/SpireBoss/Audio/Settings/ATT_MajorSpire_Attacks_Short
 bOverrideConcurrency=true
 maxDistance=11000
 soundClassObject=/Game/Sounds/FortSoundClassesAndModes/Fort_BossFight_SoundClass/Fort_BossFight_Foley
 volumeMultiplier=1.0024
 concurrencyOverrides {bLimitToOwner=true, maxCount=1}
}
