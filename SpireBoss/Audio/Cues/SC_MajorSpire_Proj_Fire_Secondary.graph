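# Sound cue for the Major Spire boss's secondary projectile fire sound.
# A Random node picks one of four equally weighted wave variations and feeds a default-setting Modulator for pitch and volume variation.
# Output uses a boosted volume, short-range attack attenuation, the boss-fight projectile sound class and the shared in-air projectile concurrency set.
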SoundCue /SpireBoss/Audio/Cues/SC_MajorSpire_Proj_Fire_Secondary
root=Modulator_0
nodes [
 {id=Modulator_0, children [Random_1]}
 {
  id=Random_1
  children [WavePlayer_2,WavePlayer_6,WavePlayer_7,WavePlayer_8]
  properties {bShouldExcludeFromBranchCulling=true, weights [1,1,1,1]}
 }
 {id=WavePlayer_2, properties {soundWaveAssetPtr=/SpireBoss/Audio/Waves/MajorSpire_Proj_Fire_Secondary_01}}
 {id=WavePlayer_6, properties {soundWaveAssetPtr=/SpireBoss/Audio/Waves/MajorSpire_Proj_Fire_Secondary_02}}
 {id=WavePlayer_7, properties {soundWaveAssetPtr=/SpireBoss/Audio/Waves/MajorSpire_Proj_Fire_Secondary_03}}
 {id=WavePlayer_8, properties {soundWaveAssetPtr=/SpireBoss/Audio/Waves/MajorSpire_Proj_Fire_Secondary_04}}
]
properties {
 attenuationSettings=/SpireBoss/Audio/Settings/ATT_MajorSpire_Attacks_Short
 maxDistance=11000
 priority=4
 soundClassObject=/Game/Sounds/FortSoundClassesAndModes/Fort_BossFight_SoundClass/Fort_BossFight_Projectile
 volumeMultiplier=2
 concurrencyOverrides {maxCount=3}
 concurrencySet [/SpireBoss/Audio/Settings/Conc_MajorSpire_Proj_InAir]
}
