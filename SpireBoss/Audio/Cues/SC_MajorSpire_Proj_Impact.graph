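# Sound cue for the Major Spire boss's projectile impact, routed to the boss-fight player-critical sound class.
# A Random node picks one of four equally weighted impact waves and feeds a default Modulator for slight pitch and volume variation.
# It uses short-range attack attenuation and a boosted volume.
# Its own concurrency override caps simultaneous instances, taking precedence over the assigned projectile concurrency set.
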SoundCue /SpireBoss/Audio/Cues/SC_MajorSpire_Proj_Impact
root=Modulator_0
nodes [
 {id=Modulator_0, children [Random_0]}
 {
  id=Random_0
  children [WavePlayer_0,WavePlayer_1,WavePlayer_3,WavePlayer_4]
  properties {bShouldExcludeFromBranchCulling=true, weights [1,1,1,1]}
 }
 {id=WavePlayer_0, properties {soundWaveAssetPtr=/SpireBoss/Audio/Waves/MajorSpire_Proj_Impact_01}}
 {id=WavePlayer_1, properties {soundWaveAssetPtr=/SpireBoss/Audio/Waves/MajorSpire_Proj_Impact_02}}
 {id=WavePlayer_3, properties {soundWaveAssetPtr=/SpireBoss/Audio/Waves/MajorSpire_Proj_Impact_03}}
 {id=WavePlayer_4, properties {soundWaveAssetPtr=/SpireBoss/Audio/Waves/MajorSpire_Proj_Impact_04}}
]
properties {
 attenuationSettings=/SpireBoss/Audio/Settings/ATT_MajorSpire_Attacks_Short
 bOverrideConcurrency=true
 maxDistance=11000
 soundClassObject=/Game/Sounds/FortSoundClassesAndModes/Fort_BossFight_SoundClass/Fort_BossFight_PlayerCritical
 volumeMultiplier=1.5
 concurrencyOverrides {maxCount=3}
 concurrencySet [/SpireBoss/Audio/Settings/Conc_MajorSpire_Proj]
}
