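# Sound cue for the Major Spire boss's projectile, combining its firing sound with the looping in-air sound.
# A Mixer layers a louder loop branch (one of three random loops with a half-second Enveloper fade-in, its own attenuation and a Modulator) over one of five random fire one-shots using short-attack attenuation.
# It uses the boss-fight projectile sound class and capped concurrency, and feeds sidechain buses for spire loops, weapons, music, ambience and footsteps.
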SoundCue /SpireBoss/Audio/Cues/SC_MajorSpire_Proj_InAir_Loop
root=Mixer_0
nodes [
 {id=Attenuation_0, children [Enveloper_0], properties {attenuationSettings=/SpireBoss/Audio/Settings/ATT_MajorSpire_Proj_InAir}}
 {id=Attenuation_1, children [Random_0], properties {attenuationSettings=/SpireBoss/Audio/Settings/ATT_MajorSpire_Attacks_Short}}
 {
  id=Enveloper_0
  children [Random_1]
  properties {pitchCurve {editorCurveData {keys [{time=0}]}}, volumeCurve {editorCurveData {keys [<at 0 0>,<at 0.5 1>]}}}
 }
 {id=Mixer_0, children [Modulator_0, Attenuation_1], properties {inputVolume [2.5,1]}}
 {id=Modulator_0, children [Attenuation_0]}
 {
  id=Random_0
  children [WavePlayer_1,WavePlayer_3,WavePlayer_4,WavePlayer_5,WavePlayer_6]
  properties {bShouldExcludeFromBranchCulling=true, weights [1,1,1,1,1]}
 }
 {id=Random_1, children [WavePlayer_0,WavePlayer_7,WavePlayer_8], properties {bShouldExcludeFromBranchCulling=true, weights [1,1,1]}}
 {id=WavePlayer_0, properties {bLooping=true, soundWaveAssetPtr=/SpireBoss/Audio/Waves/MajorSpire_Proj_InAir_Loop_01}}
 {id=WavePlayer_1, properties {soundWaveAssetPtr=/SpireBoss/Audio/Waves/MajorSpire_Proj_Fire_05}}
 {id=WavePlayer_3, properties {soundWaveAssetPtr=/SpireBoss/Audio/Waves/MajorSpire_Proj_Fire_04}}
 {id=WavePlayer_4, properties {soundWaveAssetPtr=/SpireBoss/Audio/Waves/MajorSpire_Proj_Fire_03}}
 {id=WavePlayer_5, properties {soundWaveAssetPtr=/SpireBoss/Audio/Waves/MajorSpire_Proj_Fire_02}}
 {id=WavePlayer_6, properties {soundWaveAssetPtr=/SpireBoss/Audio/Waves/MajorSpire_Proj_Fire_01}}
 {id=WavePlayer_7, properties {bLooping=true, soundWaveAssetPtr=/SpireBoss/Audio/Waves/MajorSpire_Proj_InAir_Loop_02}}
 {id=WavePlayer_8, properties {bLooping=true, soundWaveAssetPtr=/SpireBoss/Audio/Waves/MajorSpire_Proj_InAir_Loop_03}}
]
properties {
 bBypassVolumeScaleForPriority=true
 maxDistance=11000
 priority=3
 soundClassObject=/Game/Sounds/FortSoundClassesAndModes/Fort_BossFight_SoundClass/Fort_BossFight_Projectile
 volumeMultiplier=2.2
 busSends [
  {audioBus=/SpireSharedAssets/Audio/Modulation/AB_Spires_Loops_SideChain_Vol, minSendLevel=0, sendLevel=0.5, soundSourceBus=null}
  {audioBus=&AB_Sidechain_FortWeapon_Vol, minSendLevel=0, sendLevel=0.5, soundSourceBus=null}
  {audioBus=&AB_Sidechain_FortMusic_Vol, minSendLevel=0, sendLevel=0.3, soundSourceBus=null}
  {audioBus=&AB_Sidechain_FortAmbient_Vol, minSendLevel=0, sendLevel=0.8, soundSourceBus=null}
  {audioBus=/Game/Sounds/Modulation/EnvelopeFollowers/Volume/AB_Sidechain_Footsteps_Vol, minSendLevel=0, sendLevel=0.6, soundSourceBus=null}
 ]
 concurrencyOverrides {maxCount=3}
 concurrencySet [/SpireBoss/Audio/Settings/Conc_MajorSpire_Proj]
}
