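# Sound cue for the intro of the Major Spire boss's projectile attack, playing one of three slam-rise waves.
# An equally weighted Random node picks the wave and feeds a Modulator at the root.
# It uses the boss attack attenuation and the player-critical boss fight sound class.
# A concurrency override limits it to one instance per owner, taking precedence over the assigned projectile concurrency set.
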SoundCue /SpireBoss/Audio/Cues/SC_MajorSpire_Proj_Intro
root=Modulator_0
nodes [
 {id=Modulator_0, children [Random_0]}
 {id=Random_0, children [WavePlayer_0,WavePlayer_1,WavePlayer_3], properties {bShouldExcludeFromBranchCulling=true, weights [1,1,1]}}
 {id=WavePlayer_0, properties {soundWaveAssetPtr=/SpireBoss/Audio/Waves/MajorSpire_Slam_Rise_01}}
 {id=WavePlayer_1, properties {soundWaveAssetPtr=/SpireBoss/Audio/Waves/MajorSpire_Slam_Rise_02}}
 {id=WavePlayer_3, properties {soundWaveAssetPtr=/SpireBoss/Audio/Waves/MajorSpire_Slam_Rise_03}}
]
properties {
 attenuationSettings=/SpireBoss/Audio/Settings/ATT_MajorSpire_Attacks
 bOverrideConcurrency=true
 maxDistance=17000
 soundClassObject=/Game/Sounds/FortSoundClassesAndModes/Fort_BossFight_SoundClass/Fort_BossFight_PlayerCritical
 volumeMultiplier=1.0024
 concurrencyOverrides {bLimitToOwner=true, maxCount=1}
 concurrencySet [/SpireBoss/Audio/Settings/Conc_MajorSpire_Proj]
}
