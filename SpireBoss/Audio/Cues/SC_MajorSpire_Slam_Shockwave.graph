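# Sound cue for the Major Spire boss's slam shockwave, classed as player-critical boss fight audio.
# A Random node picks one of three equally weighted shockwave waves, which feeds a Modulator for per-play variation.
# Dedicated attenuation and a slam-attack concurrency asset govern playback.
# Sends to sidechain buses for spire loops, weapons, music, ambience and footsteps let the hit drive their volume modulation.
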SoundCue /SpireBoss/Audio/Cues/SC_MajorSpire_Slam_Shockwave
root=Modulator_0
nodes [
 {id=Modulator_0, children [Random_1]}
 {id=Random_1, children [WavePlayer_2,WavePlayer_4,WavePlayer_5], properties {bShouldExcludeFromBranchCulling=true, weights [1,1,1]}}
 {id=WavePlayer_2, properties {soundWaveAssetPtr=/SpireBoss/Audio/Waves/MajorSpire_Slam_Shockwave_01}}
 {id=WavePlayer_4, properties {soundWaveAssetPtr=/SpireBoss/Audio/Waves/MajorSpire_Slam_Shockwave_02}}
 {id=WavePlayer_5, properties {soundWaveAssetPtr=/SpireBoss/Audio/Waves/MajorSpire_Slam_Shockwave_03}}
]
properties {
 attenuationSettings=/SpireBoss/Audio/Settings/ATT_MajorSpire_Slam_Shockwave
 maxDistance=3200
 priority=5
 soundClassObject=/Game/Sounds/FortSoundClassesAndModes/Fort_BossFight_SoundClass/Fort_BossFight_PlayerCritical
 volumeMultiplier=1.5
 busSends [
  {audioBus=/SpireSharedAssets/Audio/Modulation/AB_Spires_Loops_SideChain_Vol, minSendLevel=0, sendLevel=0.6, soundSourceBus=null}
  {audioBus=&AB_Sidechain_FortWeapon_Vol, minSendLevel=0, sendLevel=0.6, soundSourceBus=null}
  {audioBus=&AB_Sidechain_FortMusic_Vol, minSendLevel=0, sendLevel=0.2, soundSourceBus=null}
  {audioBus=&AB_Sidechain_FortAmbient_Vol, minSendLevel=0, sendLevel=0.8, soundSourceBus=null}
  {audioBus=/Game/Sounds/Modulation/EnvelopeFollowers/Volume/AB_Sidechain_Footsteps_Vol, minSendLevel=0, sendLevel=0.6, soundSourceBus=null}
 ]
 concurrencyOverrides {bLimitToOwner=true, maxCount=1}
 concurrencySet [/SpireBoss/Audio/Settings/Conc_MajorSpire_Attack_Slam_Pri]
}
