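# Sound cue for the Major Spire boss's slam telegraph sound, routed to the player-critical boss fight sound class.
# A Random node picks one of four equally weighted wave variants and feeds a Modulator left at default settings.
# It uses the Major Spire attack attenuation asset at reduced volume.
# A concurrency override limits it to one instance per owner.
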SoundCue /SpireBoss/Audio/Cues/SC_MajorSpire_Slam_Telegraph
root=Modulator_0
nodes [
 {id=Modulator_0, children [Random_1]}
 {
  id=Random_1
  children [WavePlayer_2,WavePlayer_5,WavePlayer_6,WavePlayer_7]
  properties {bShouldExcludeFromBranchCulling=true, weights [1,1,1,1]}
 }
 {id=WavePlayer_2, properties {soundWaveAssetPtr=/SpireBoss/Audio/Waves/MajorSpire_Slam_Telegraph_01}}
 {id=WavePlayer_5, properties {soundWaveAssetPtr=/SpireBoss/Audio/Waves/MajorSpire_Slam_Telegraph_02}}
 {id=WavePlayer_6, properties {soundWaveAssetPtr=/SpireBoss/Audio/Waves/MajorSpire_Slam_Telegraph_03}}
 {id=WavePlayer_7, properties {soundWaveAssetPtr=/SpireBoss/Audio/Waves/MajorSpire_Slam_Telegraph_04}}
]
properties {
 attenuationSettings=/SpireBoss/Audio/Settings/ATT_MajorSpire_Attacks
 bOverrideConcurrency=true
 maxDistance=17000
 soundClassObject=/Game/Sounds/FortSoundClassesAndModes/Fort_BossFight_SoundClass/Fort_BossFight_PlayerCritical
 volumeMultiplier=0.5012
 concurrencyOverrides {bLimitToOwner=true, maxCount=1}
}
