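# Sound cue for the Major Spire boss's summon sound, routed to the player-critical boss fight sound class.
# A Random node picks one of five equally weighted summon waves and feeds a Modulator that adds slight pitch variation.
# It uses the Major Spire attacks attenuation and overrides concurrency to one voice, stopping the oldest.
# Bus sends feed sidechain volume buses for spire loops, weapons, music, ambience and footsteps.
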SoundCue /SpireBoss/Audio/Cues/SC_MajorSpire_Summon
root=Modulator_0
nodes [
 {id=Modulator_0, children [Random_1], properties {pitchMax=1, pitchMin=0.98}}
 {
  id=Random_1
  children [WavePlayer_3,WavePlayer_4,WavePlayer_5,WavePlayer_6,WavePlayer_7]
  properties {bShouldExcludeFromBranchCulling=true, weights [1,1,1,1,1]}
 }
 {id=WavePlayer_3, properties {soundWaveAssetPtr=/SpireBoss/Audio/Waves/MajorSpire_Summon_05}}
 {id=WavePlayer_4, properties {soundWaveAssetPtr=/SpireBoss/Audio/Waves/MajorSpire_Summon_04}}
 {id=WavePlayer_5, properties {soundWaveAssetPtr=/SpireBoss/Audio/Waves/MajorSpire_Summon_03}}
 {id=WavePlayer_6, properties {soundWaveAssetPtr=/SpireBoss/Audio/Waves/MajorSpire_Summon_02}}
 {id=WavePlayer_7, properties {soundWaveAssetPtr=/SpireBoss/Audio/Waves/MajorSpire_Summon_01}}
]
properties {
 attenuationSettings=/SpireBoss/Audio/Settings/ATT_MajorSpire_Attacks
 bOverrideConcurrency=true
 maxDistance=17000
 soundClassObject=/Game/Sounds/FortSoundClassesAndModes/Fort_BossFight_SoundClass/Fort_BossFight_PlayerCritical
 volumeMultiplier=2.3
 attenuationOverrides {bAttenuate=false, falloffDistance=10000, attenuationShapeExtents {x=0}}
 busSends [
  {audioBus=/SpireSharedAssets/Audio/Modulation/AB_Spires_Loops_SideChain_Vol, minSendLevel=0, sendLevel=0.8, soundSourceBus=null}
  {audioBus=&AB_Sidechain_FortWeapon_Vol, minSendLevel=0, sendLevel=0.6, soundSourceBus=null}
  {audioBus=&AB_Sidechain_FortMusic_Vol, minSendLevel=0, sendLevel=0.4, soundSourceBus=null}
  {audioBus=&AB_Sidechain_FortAmbient_Vol, minSendLevel=0, sendLevel=0.8, soundSourceBus=null}
  {audioBus=/Game/Sounds/Modulation/EnvelopeFollowers/Volume/AB_Sidechain_Footsteps_Vol, minSendLevel=0, sendLevel=0.8, soundSourceBus=null}
 ]
 concurrencyOverrides {maxCount=1, resolutionRule=StopOldest, retriggerTime=0.5, voiceStealReleaseTime=0.1}
}
