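# Sound cue for the end of the Major Spire boss's summon, routed to the boss-fight player-critical sound class.
# A Random node picks one of five equally weighted wave variations, feeding a Modulator that slightly varies pitch downward.
# It uses the short-range Major Spire attack attenuation asset with a boosted cue volume.
# Its own concurrency override caps it at two instances, stopping the oldest, with a retrigger time and brief voice-steal release.
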SoundCue /SpireBoss/Audio/Cues/SC_MajorSpire_Summon_End
root=Modulator_0
nodes [
 {id=Modulator_0, children [Random_0], properties {pitchMax=1, pitchMin=0.98}}
 {
  id=Random_0
  children [WavePlayer_0,WavePlayer_1,WavePlayer_2,WavePlayer_8,WavePlayer_9]
  properties {bShouldExcludeFromBranchCulling=true, weights [1,1,1,1,1]}
 }
 {id=WavePlayer_0, properties {soundWaveAssetPtr=/SpireBoss/Audio/Waves/MajorSpire_Summon_End_01}}
 {id=WavePlayer_1, properties {soundWaveAssetPtr=/SpireBoss/Audio/Waves/MajorSpire_Summon_End_02}}
 {id=WavePlayer_2, properties {soundWaveAssetPtr=/SpireBoss/Audio/Waves/MajorSpire_Summon_End_03}}
 {id=WavePlayer_8, properties {soundWaveAssetPtr=/SpireBoss/Audio/Waves/MajorSpire_Summon_End_04}}
 {id=WavePlayer_9, properties {soundWaveAssetPtr=/SpireBoss/Audio/Waves/MajorSpire_Summon_End_05}}
]
properties {
 attenuationSettings=/SpireBoss/Audio/Settings/ATT_MajorSpire_Attacks_Short
 bOverrideConcurrency=true
 maxDistance=11000
 soundClassObject=/Game/Sounds/FortSoundClassesAndModes/Fort_BossFight_SoundClass/Fort_BossFight_PlayerCritical
 volumeMultiplier=1.2
 attenuationOverrides {bAttenuate=false, falloffDistance=10000, attenuationShapeExtents {x=0}}
 concurrencyOverrides {maxCount=2, resolutionRule=StopOldest, retriggerTime=0.5, voiceStealReleaseTime=0.1}
}
