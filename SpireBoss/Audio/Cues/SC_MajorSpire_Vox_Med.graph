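# Sound cue for the Major Spire boss's medium vocalizations, played on the boss-fight player-critical sound class with boss vocal attenuation settings.
# A quality-level switch feeds the first two levels from a random pick of seven waves and the last two from a four-wave subset.
# A root modulator randomly raises pitch slightly.
# A concurrency override caps simultaneous voices at a small count, stopping the oldest.
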SoundCue /SpireBoss/Audio/Cues/SC_MajorSpire_Vox_Med
root=Modulator_0
nodes [
 {id=Modulator_0, children [QualityLevel_0], properties {pitchMax=1.2, pitchMin=1}}
 {id=QualityLevel_0, children [Random_1,Random_1,Random_2,Random_2]}
 {
  id=Random_1
  children [WavePlayer_3, WavePlayer_4, WavePlayer_5, WavePlayer_6, WavePlayer_7, WavePlayer_9, WavePlayer_10]
  properties {bShouldExcludeFromBranchCulling=true, weights [1,1,1,1,1,1,1]}
 }
 {id=Random_2, children [WavePlayer_11, WavePlayer_12, WavePlayer_13, WavePlayer_14], properties {weights [1,1,1,1]}}
 {id=WavePlayer_10, properties {soundWaveAssetPtr=/SpireBoss/Audio/Waves/MajorSpire_Vox_Med_01}}
 {id=WavePlayer_11, properties {soundWaveAssetPtr=/SpireBoss/Audio/Waves/MajorSpire_Vox_Med_07}}
 {id=WavePlayer_12, properties {soundWaveAssetPtr=/SpireBoss/Audio/Waves/MajorSpire_Vox_Med_06}}
 {id=WavePlayer_13, properties {soundWaveAssetPtr=/SpireBoss/Audio/Waves/MajorSpire_Vox_Med_05}}
 {id=WavePlayer_14, properties {soundWaveAssetPtr=/SpireBoss/Audio/Waves/MajorSpire_Vox_Med_04}}
 {id=WavePlayer_3, properties {soundWaveAssetPtr=&MajorSpire_Vox_Med_07}}
 {id=WavePlayer_4, properties {soundWaveAssetPtr=&MajorSpire_Vox_Med_06}}
 {id=WavePlayer_5, properties {soundWaveAssetPtr=&MajorSpire_Vox_Med_05}}
 {id=WavePlayer_6, properties {soundWaveAssetPtr=&MajorSpire_Vox_Med_04}}
 {id=WavePlayer_7, properties {soundWaveAssetPtr=/SpireBoss/Audio/Waves/MajorSpire_Vox_Med_03}}
 {id=WavePlayer_9, properties {soundWaveAssetPtr=/SpireBoss/Audio/Waves/MajorSpire_Vox_Med_02}}
]
properties {
 attenuationSettings=/SpireBoss/Audio/Settings/ATT_MajorSpire_Vocals
 bOverrideConcurrency=true
 maxDistance=13500
 soundClassObject=/Game/Sounds/FortSoundClassesAndModes/Fort_BossFight_SoundClass/Fort_BossFight_PlayerCritical
 volumeMultiplier=2.5
 attenuationOverrides {bAttenuate=false, falloffDistance=10000, attenuationShapeExtents {x=0}}
 concurrencyOverrides {maxCount=2, resolutionRule=StopOldest, retriggerTime=0.5, voiceStealReleaseTime=0.1}
}
